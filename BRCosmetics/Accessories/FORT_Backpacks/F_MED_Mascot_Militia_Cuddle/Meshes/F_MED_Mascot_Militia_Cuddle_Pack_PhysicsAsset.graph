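# Physics asset for the Mascot Militia Cuddle backpack cosmetic, providing bodies and joints for its two dangling pouches and zipper pull.
# The pack body combines a capsule and two tilted boxes, while each dynamic part has a box body and a tiny master body.
# Master bodies join the pack with free angular motion, and each part hinges off its master with narrow twist and wide second swing, with the zipper's first swing locked.
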PhysicsAsset /BRCosmetics/Accessories/FORT_Backpacks/F_MED_Mascot_Militia_Cuddle/Meshes/F_MED_Mascot_Militia_Cuddle_Pack_PhysicsAsset
bodies {
 dyn_pouch_l [{shapeType=Box, extentX=15.9017, extentY=8.6774, extentZ=3.8027, center=<0.4341 -0.1146 -2.5042>}]
 dyn_pouch_l_mstr [{shapeType=Box, extentX=1.01, extentY=1.01, extentZ=1.01}]
 dyn_pouch_r [{shapeType=Box, extentX=15.9957, extentY=7.9394, extentZ=3.7011, center=<0.8802 0.342 -2.3325>}]
 dyn_pouch_r_mstr [{shapeType=Box, extentX=1.01, extentY=1.01, extentZ=1.01}]
 dyn_zipper [{shapeType=Box, extentX=3.4986, extentY=1.7957, extentZ=0.6123, center=<1.5999 -0.0128 -0.0345>}]
 dyn_zipper_mstr [{shapeType=Box, extentX=1.01, extentY=1.01, extentZ=1.01}]
 pack [
  {shapeType=Capsule, length=15.8065, radius=7.7507, center=<0 0 3.314>}
  {shapeType=Box, extentX=7.2849, extentY=12.5873, extentZ=11.7497, center=<-6.4756 0 -1.3783>, rotation=<rot 19.9999 0 0>}
  {shapeType=Box, extentX=6.5846, extentY=12.5873, extentZ=19.9246, center=<7.2535 0 1.738>, rotation=<rot -20.0001 0 0>}
 ]
}
constraints [
 {
  bone1Name=dyn_pouch_l_mstr
  bone2Name=pack
  swing1LimitDegrees=45
  swing1Motion=Free
  swing2LimitDegrees=45
  swing2Motion=Free
  twistLimitDegrees=45
  twistMotion=Free
 }
 {
  bone1Name=dyn_pouch_r_mstr
  bone2Name=pack
  swing1LimitDegrees=45
  swing1Motion=Free
  swing2LimitDegrees=45
  swing2Motion=Free
  twistLimitDegrees=45
  twistMotion=Free
 }
 {
  bone1Name=dyn_zipper_mstr
  bone2Name=pack
  swing1LimitDegrees=45
  swing1Motion=Free
  swing2LimitDegrees=45
  swing2Motion=Free
  twistLimitDegrees=45
  twistMotion=Free
 }
 {bone1Name=dyn_pouch_r, bone2Name=dyn_pouch_r_mstr, swing1LimitDegrees=10, swing2LimitDegrees=90, twistLimitDegrees=5}
 {bone1Name=dyn_zipper, bone2Name=dyn_zipper_mstr, swing1LimitDegrees=10, swing1Motion=Locked, swing2LimitDegrees=100, twistLimitDegrees=5}
 {bone1Name=dyn_pouch_l, bone2Name=dyn_pouch_l_mstr, swing1LimitDegrees=10, swing2LimitDegrees=90, twistLimitDegrees=5}
]
